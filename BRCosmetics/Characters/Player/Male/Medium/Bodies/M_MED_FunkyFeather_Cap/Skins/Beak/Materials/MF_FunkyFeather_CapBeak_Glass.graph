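# Material function for the FunkyFeather Cap Beak skin that layers a bubbly fake-glass look onto input material attributes through a selectable mask channel.
# A custom scene-color blur, scaled by camera distance and dither, supplies see-through color alongside Fresnel tinting and thin-film emissive with time-of-day scaling.
# Animated bubbles use distorted, parallax-offset pre-skinned projection, masked by view-aligned distortion.
# Feature-level, shading-path and quality switches provide cheaper mobile and low-end fallbacks.
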MaterialFunction /BRCosmetics/Characters/Player/Male/Medium/Bodies/M_MED_FunkyFeather_Cap/Skins/Beak/Materials/MF_FunkyFeather_CapBeak_Glass
graph {expressions {
  MaterialExpressionAbs_0 {class=MaterialExpressionAbs, inputs [None {from=MaterialExpressionComponentMask_0, output=""}], properties {}}
  MaterialExpressionAdd_15 {
   class=MaterialExpressionAdd
   inputs [
    A {from=MaterialExpressionReroute_2}
    B {from=MaterialExpressionMultiply_34}
   ]
  }
  MaterialExpressionAdd_17 {
   class=MaterialExpressionAdd
   inputs [
    A {from=MaterialExpressionShadingPathSwitch_0}
    B {from=MaterialExpressionMultiply_35}
   ]
  }
  MaterialExpressionAdd_4 {
   class=MaterialExpressionAdd
   inputs [
    A {from=MaterialExpressionMultiply_29}
    B {from=MaterialExpressionNamedRerouteUsage_5}
   ]
  }
  MaterialExpressionAdd_7 {
   class=MaterialExpressionAdd
   inputs [
    A {from=MaterialExpressionPanner_3}
    B {from=MaterialExpressionConstantBiasScale_0}
   ]
  }
  MaterialExpressionAdd_8 {
   class=MaterialExpressionAdd
   inputs [
    A {from=MaterialExpressionConvert_2}
    B {from=MaterialExpressionMultiply_36}
   ]
  }
  MaterialExpressionAppendVector_2 {
   class=MaterialExpressionAppendVector
   inputs [
    A {from=MaterialExpressionScalarParameter_28, output=""}
    B {from=MaterialExpressionScalarParameter_29, output=""}
   ]
  }
  MaterialExpressionBlendMaterialAttributes_0 {
   class=MaterialExpressionBlendMaterialAttributes
   inputs [
    A {from=MaterialExpressionSetMaterialAttributes_1, output=""}
    B {from=MaterialExpressionNamedRerouteUsage_25, output=""}
    Alpha {from=MaterialExpressionChannelMaskParameter_0, output=R}
   ]
   properties {}
  }
  MaterialExpressionCameraPositionWS_1 {class=MaterialExpressionCameraPositionWS, inputs [], properties {}}
  MaterialExpressionChannelMaskParameter_0 {
   class=MaterialExpressionChannelMaskParameter
   inputs [None {from=MaterialExpressionNamedRerouteUsage_24, output=""}]
   properties {group=FunkyFeather Beak, parameterName=TattooChannel}
  }
  MaterialExpressionComponentMask_0 {class=MaterialExpressionComponentMask, inputs [None {from=MaterialExpressionNamedRerouteUsage_36}], properties {b=true}}
  MaterialExpressionComponentMask_4 {class=MaterialExpressionComponentMask, inputs [None {from=MaterialExpressionMultiply_37}], properties {g=true}}
  MaterialExpressionComponentMask_5 {class=MaterialExpressionComponentMask, inputs [None {from=MaterialExpressionTransformPosition_1}], properties {b=true}}
  MaterialExpressionConstant2Vector_2 {class=MaterialExpressionConstant2Vector, properties {g=0.1}}
  MaterialExpressionConstantBiasScale_0 {class=MaterialExpressionConstantBiasScale, inputs [None {from=MaterialExpressionSine_0}], properties {scale=0.1}}
  MaterialExpressionConstant_10 {class=MaterialExpressionConstant, properties {r=1}}
  MaterialExpressionConstant_11 {class=MaterialExpressionConstant}
  MaterialExpressionConstant_9 {class=MaterialExpressionConstant, properties {r=1}}
  MaterialExpressionConvert_2 {
   class=MaterialExpressionConvert
   inputs [None {from=MaterialExpressionDivide_3}]
   properties {
    nodeName=Break Float3
    convertInputs [{
      type=Vector3
      expressionInput {
       expression=&self:MaterialExpressionDivide_3
       inputName=None
       mask=0
       maskA=0
       maskB=0
       maskG=0
       maskR=0
       outputIndex=0
      }
     }]
    convertOutputs [{type=Scalar}, {type=Scalar}, {type=Scalar}]
   }
  }
  MaterialExpressionConvert_3 {
   class=MaterialExpressionConvert
   inputs [
    None {from=MaterialExpressionConvert_2}
    None {from=MaterialExpressionConvert_2}
   ]
   properties {
    nodeName=Make Vector2
    convertInputs [
     {
      type=Scalar
      expressionInput {
       expression=&self:MaterialExpressionConvert_2
       inputName=None
       mask=0
       maskA=0
       maskB=0
       maskG=0
       maskR=0
       outputIndex=0
      }
     }
     {
      type=Scalar
      expressionInput {
       expression=&self:MaterialExpressionConvert_2
       inputName=None
       mask=0
       maskA=0
       maskB=0
       maskG=0
       maskR=0
       outputIndex=1
      }
     }
    ]
    convertOutputs [{type=Vector2}]
   }
  }
  MaterialExpressionCustom_1 {
   class=MaterialExpressionCustom
   inputs [
    stepSize {from=MaterialExpressionNamedRerouteDeclaration_31, output=""}
    iterations {from=MaterialExpressionQualitySwitch_2, output=""}
   ]
   properties {
    code="float dist = stepSize / iterations;\r\nint BufferIndex = 2;\r\nfloat2 offsets[8] = { float2(1,0), float2(-1,0), float2(0,1), float2(0,-1),float2(.5,.5), float2(-.5,-.5), float2(-.5,.5), float2(.5,-.5)};\r\nfloat2 UV = GetDefaultSceneTextureUV(Parameters, int(BufferIndex));\r\nfloat2 NewUV = UV;\r\nfloat3 CurColor=0;\r\nfloat2 ratio = float2(1, View.BufferSizeAndInvSize.x / View.BufferSizeAndInvSize.y); //take in account of aspect ratio\r\n\r\nfor (int i = 0; i < (int)iterations; i++)\r\n{\r\n\tfor (int j = 0; j < 8; j++)\r\n\t{\r\n\t\tNewUV = UV + offsets[j] * ratio * (i+1) * dist;\r\n\t\t//CurColor += SceneTextureLookupCommon(Parameters,NewUV, BufferIndex, false);\r\n        CurColor += DecodeSceneColorForMaterialNode(NewUV);\r\n\t}\r\n}\r\n\r\nCurColor /= (iterations * 8.0);\r\nreturn CurColor;\r\n"
    showCode=true
    inputs [
     {
      inputName=stepSize
      input {
       expression=&self:MaterialExpressionNamedRerouteDeclaration_31
       inputName=None
       mask=0
       maskA=0
       maskB=0
       maskG=0
       maskR=0
       outputIndex=0
      }
     }
     {
      inputName=iterations
      input {
       expression=&self:MaterialExpressionQualitySwitch_2
       inputName=None
       mask=0
       maskA=0
       maskB=0
       maskG=0
       maskR=0
       outputIndex=0
      }
     }
    ]
   }
  }
  MaterialExpressionDesaturation_0 {
   class=MaterialExpressionDesaturation
   inputs [
    None {from=MaterialExpressionLinearInterpolate_3, output=""}
    Fraction {from=MaterialExpressionScalarParameter_9, output=""}
   ]
   properties {}
  }
  MaterialExpressionDesaturation_1 {
   class=MaterialExpressionDesaturation
   inputs [
    None {from=MaterialExpressionCustom_1, output=""}
    Fraction {from=MaterialExpressionScalarParameter_53, output=""}
   ]
   properties {}
  }
  MaterialExpressionDivide_3 {
   class=MaterialExpressionDivide
   inputs [
    A {from=MaterialExpressionNamedRerouteUsage_29}
    B {from=""}
   ]
   properties {constB=64}
  }
  MaterialExpressionDivide_7 {
   class=MaterialExpressionDivide
   inputs [
    A {from=MaterialExpressionSubtract_4}
    B {from=MaterialExpressionScalarParameter_48}
   ]
  }
  MaterialExpressionFeatureLevelSwitch_1 {
   class=MaterialExpressionFeatureLevelSwitch
   inputs [
    Default {from=MaterialExpressionMultiply_6, output=""}
    ES2 {from=MaterialExpressionScalarParameter_15, output=""}
    ES3_1 {from=MaterialExpressionScalarParameter_15, output=""}
    SM4_REMOVED {from="", output=""}
    SM5 {from=MaterialExpressionMultiply_6, output=""}
    SM6 {from=MaterialExpressionMultiply_6, output=""}
   ]
   properties {}
  }
  MaterialExpressionFeatureLevelSwitch_2 {
   class=MaterialExpressionFeatureLevelSwitch
   inputs [
    Default {from=MaterialExpressionMultiply_15, output=""}
    ES2 {from=MaterialExpressionVectorParameter_2, output=RGB}
    ES3_1 {from=MaterialExpressionVectorParameter_2, output=RGB}
    SM4_REMOVED {from="", output=""}
    SM5 {from=MaterialExpressionMultiply_15, output=""}
    SM6 {from=MaterialExpressionMultiply_15, output=""}
   ]
   properties {}
  }
  MaterialExpressionFeatureLevelSwitch_3 {
   class=MaterialExpressionFeatureLevelSwitch
   inputs [
    Default {from=MaterialExpressionMultiply_0, output=""}
    ES2 {from=MaterialExpressionVectorParameter_3, output=RGB}
    ES3_1 {from=MaterialExpressionVectorParameter_3, output=RGB}
    SM4_REMOVED {from="", output=""}
    SM5 {from=MaterialExpressionMultiply_0, output=""}
    SM6 {from=MaterialExpressionMultiply_0, output=""}
   ]
   properties {}
  }
  MaterialExpressionFresnel_0 {
   class=MaterialExpressionFresnel
   inputs [
    ExponentIn {from=MaterialExpressionScalarParameter_45, output=""}
    BaseReflectFractionIn {from="", output=""}
    Normal {from="", output=""}
   ]
   properties {baseReflectFraction=0, exponent=1}
  }
  MaterialExpressionFresnel_1 {
   class=MaterialExpressionFresnel
   inputs [
    ExponentIn {from=MaterialExpressionScalarParameter_14, output=""}
    BaseReflectFractionIn {from="", output=""}
    Normal {from="", output=""}
   ]
   properties {baseReflectFraction=0, exponent=3}
  }
  MaterialExpressionFresnel_2 {
   class=MaterialExpressionFresnel
   inputs [
    ExponentIn {from="", output=""}
    BaseReflectFractionIn {from="", output=""}
    Normal {from="", output=""}
   ]
   properties {exponent=0.2}
  }
  MaterialExpressionFunctionInput_0 {
   class=MaterialExpressionFunctionInput
   inputs [Preview {from=MaterialExpressionShaderStageSwitch_0, output=""}]
   properties {inputName=PreskinnedLocPos, sortPriority=1}
  }
  MaterialExpressionFunctionInput_1 {
   class=MaterialExpressionFunctionInput
   inputs [Preview {from="", output=""}]
   properties {inputName=Material, inputType=FunctionInput_MaterialAttributes}
  }
  MaterialExpressionFunctionInput_2 {class=MaterialExpressionFunctionInput, inputs [Preview {from="", output=""}], properties {inputName=TechArtMask, sortPriority=2}}
  MaterialExpressionFunctionOutput_0 {class=MaterialExpressionFunctionOutput, inputs [None {from=MaterialExpressionBlendMaterialAttributes_0, output=""}]}
  MaterialExpressionGetMaterialAttributes_1 {
   class=MaterialExpressionGetMaterialAttributes
   inputs [None {from=MaterialExpressionFunctionInput_1, output=""}]
   properties {attributeGetTypes [
     #1
     #2
     #3
     #4
     #5
     #6
    ]}
  }
  MaterialExpressionLinearInterpolate_0 {
   class=MaterialExpressionLinearInterpolate
   inputs [
    A {from=MaterialExpressionNamedRerouteUsage_10}
    B {from=""}
    Alpha {from=MaterialExpressionNamedRerouteDeclaration_33}
   ]
  }
  MaterialExpressionLinearInterpolate_1 {
   class=MaterialExpressionLinearInterpolate
   inputs [
    A {from=MaterialExpressionNamedRerouteUsage_8}
    B {from=MaterialExpressionVectorParameter_8, output=RGB}
    Alpha {from=MaterialExpressionNamedRerouteDeclaration_33}
   ]
  }
  MaterialExpressionLinearInterpolate_11 {
   class=MaterialExpressionLinearInterpolate
   inputs [
    A {from=MaterialExpressionLinearInterpolate_1}
    B {from=MaterialExpressionVectorParameter_6, output=RGB}
    Alpha {from=MaterialExpressionNamedRerouteUsage_30}
   ]
  }
  MaterialExpressionLinearInterpolate_12 {
   class=MaterialExpressionLinearInterpolate
   inputs [
    A {from=MaterialExpressionQualitySwitch_1}
    B {from=MaterialExpressionScalarParameter_49}
    Alpha {from=MaterialExpressionSaturate_0}
   ]
  }
  MaterialExpressionLinearInterpolate_13 {
   class=MaterialExpressionLinearInterpolate
   inputs [
    A {from=MaterialExpressionLinearInterpolate_12}
    B {from=MaterialExpressionMultiply_40}
    Alpha {from=MaterialExpressionMaterialFunctionCall_10, output=Result}
   ]
  }
  MaterialExpressionLinearInterpolate_14 {
   class=MaterialExpressionLinearInterpolate
   inputs [
    A {from=MaterialExpressionLinearInterpolate_12}
    B {from=MaterialExpressionLinearInterpolate_13}
    Alpha {from=MaterialExpressionScalarParameter_50}
   ]
  }
  MaterialExpressionLinearInterpolate_3 {
   class=MaterialExpressionLinearInterpolate
   inputs [
    A {from=""}
    B {from=MaterialExpressionTextureSampleParameter2D_0, output=RGB}
    Alpha {from=MaterialExpressionFresnel_1}
   ]
  }
  MaterialExpressionLinearInterpolate_4 {
   class=MaterialExpressionLinearInterpolate
   inputs [
    A {from=MaterialExpressionVectorParameter_0, output=RGB}
    B {from=""}
    Alpha {from=MaterialExpressionFresnel_2}
   ]
   properties {constB=0}
  }
  MaterialExpressionLinearInterpolate_5 {
   class=MaterialExpressionLinearInterpolate
   inputs [
    A {from=""}
    B {from=MaterialExpressionTextureSample_2, output=B}
    Alpha {from=MaterialExpressionScalarParameter_4}
   ]
   properties {constA=1}
  }
  MaterialExpressionLinearInterpolate_9 {
   class=MaterialExpressionLinearInterpolate
   inputs [
    A {from=MaterialExpressionLinearInterpolate_0}
    B {from=""}
    Alpha {from=MaterialExpressionNamedRerouteUsage_30}
   ]
  }
  MaterialExpressionLocalPosition_1 {class=MaterialExpressionLocalPosition, inputs [], properties {localOrigin=InstancePreSkinning}}
  MaterialExpressionMaterialFunctionCall_10 {
   class=MaterialExpressionMaterialFunctionCall
   inputs [
    Alpha Threshold {from=""}
    Random {from=""}
   ]
   properties {materialFunction=&DitherTemporalAA}
  }
  MaterialExpressionMaterialFunctionCall_11 {
   class=MaterialExpressionMaterialFunctionCall
   inputs [
    ReflectionVector {from=""}
    ReflectionTexture {from=""}
   ]
   properties {materialFunction=/Engine/Functions/Engine_MaterialFunctions01/Reflections/ViewAlignedReflection}
  }
  MaterialExpressionMaterialFunctionCall_6 {
   class=MaterialExpressionMaterialFunctionCall
   inputs [
    TextureObject {from=MaterialExpressionTextureObject_0}
    TextureSize {from=MaterialExpressionScalarParameter_40}
    WorldPosition {from=MaterialExpressionAdd_17}
    FadingGradient {from=""}
    In {from=MaterialExpressionConstant2Vector_2}
    World Space Normal {from=""}
    ProjectionTransitionContrast {from=MaterialExpressionScalarParameter_44}
   ]
   properties {materialFunction=/BRCosmetics/Characters/Player/Male/Medium/Bodies/M_MED_FloralMane_Sand/Skins/SuperLevels/ParentMaterials/Preskinned_WP_Texture_ScalePan_XY}
  }
  MaterialExpressionMaterialFunctionCall_8 {
   class=MaterialExpressionMaterialFunctionCall
   inputs [
    Value_Morning {from=MaterialExpressionScalarParameter_17}
    Value_Day {from=MaterialExpressionScalarParameter_17}
    Value_Evening {from=MaterialExpressionScalarParameter_17}
    Value_Night {from=MaterialExpressionScalarParameter_18}
   ]
   properties {materialFunction=&Scalar_Time_of_Day}
  }
  MaterialExpressionMax_4 {
   class=MaterialExpressionMax
   inputs [
    A {from=MaterialExpressionSceneColor_4, output=RGB}
    B {from=""}
   ]
  }
  MaterialExpressionMultiply_0 {
   class=MaterialExpressionMultiply
   inputs [
    A {from=MaterialExpressionAdd_4}
    B {from=MaterialExpressionMaterialFunctionCall_8, output=Result}
   ]
  }
  MaterialExpressionMultiply_11 {
   class=MaterialExpressionMultiply
   inputs [
    A {from=MaterialExpressionDesaturation_0}
    B {from=MaterialExpressionScalarParameter_26}
   ]
  }
  MaterialExpressionMultiply_15 {
   class=MaterialExpressionMultiply
   inputs [
    A {from=MaterialExpressionNamedRerouteUsage_6}
    B {from=MaterialExpressionVectorParameter_1, output=RGB}
   ]
  }
  MaterialExpressionMultiply_29 {
   class=MaterialExpressionMultiply
   inputs [
    A {from=MaterialExpressionNamedRerouteUsage_7}
    B {from=MaterialExpressionLinearInterpolate_4}
   ]
  }
  MaterialExpressionMultiply_34 {
   class=MaterialExpressionMultiply
   inputs [
    A {from=MaterialExpressionNormalize_1}
    B {from=MaterialExpressionScalarParameter_41}
   ]
   properties {constB=12}
  }
  MaterialExpressionMultiply_35 {
   class=MaterialExpressionMultiply
   inputs [
    A {from=MaterialExpressionNamedRerouteDeclaration_25}
    B {from=MaterialExpressionScalarParameter_42}
   ]
   properties {constB=5}
  }
  MaterialExpressionMultiply_36 {
   class=MaterialExpressionMultiply
   inputs [
    A {from=MaterialExpressionTime_0}
    B {from=""}
   ]
   properties {constB=-0.3333}
  }
  MaterialExpressionMultiply_37 {
   class=MaterialExpressionMultiply
   inputs [
    A {from=MaterialExpressionPower_2}
    B {from=MaterialExpressionReroute_3}
   ]
  }
  MaterialExpressionMultiply_38 {
   class=MaterialExpressionMultiply
   inputs [
    A {from=MaterialExpressionComponentMask_4}
    B {from=MaterialExpressionScalarParameter_43}
   ]
  }
  MaterialExpressionMultiply_40 {
   class=MaterialExpressionMultiply
   inputs [
    A {from=MaterialExpressionLinearInterpolate_12}
    B {from=MaterialExpressionScalarParameter_52}
   ]
   properties {constB=2}
  }
  MaterialExpressionMultiply_41 {
   class=MaterialExpressionMultiply
   inputs [
    A {from=MaterialExpressionMultiply_42}
    B {from=MaterialExpressionOneMinus_2}
   ]
  }
  MaterialExpressionMultiply_42 {
   class=MaterialExpressionMultiply
   inputs [
    A {from=MaterialExpressionMultiply_45}
    B {from=MaterialExpressionSmoothStep_1}
   ]
  }
  MaterialExpressionMultiply_43 {
   class=MaterialExpressionMultiply
   inputs [
    A {from=MaterialExpressionMaterialFunctionCall_11, output=UVs}
    B {from=MaterialExpressionScalarParameter_0}
   ]
  }
  MaterialExpressionMultiply_45 {
   class=MaterialExpressionMultiply
   inputs [
    A {from=MaterialExpressionLinearInterpolate_5}
    B {from=MaterialExpressionMultiply_38}
   ]
  }
  MaterialExpressionMultiply_6 {
   class=MaterialExpressionMultiply
   inputs [
    A {from=MaterialExpressionMax_4}
    B {from=MaterialExpressionScalarParameter_6}
   ]
  }
  MaterialExpressionNamedRerouteDeclaration_1 {
   class=MaterialExpressionNamedRerouteDeclaration
   inputs [None {from=MaterialExpressionGetMaterialAttributes_1, output=Normal}]
   properties {name=Normal_In, nodeColor=<rgba 0.4297 0 1 1>}
  }
  MaterialExpressionNamedRerouteDeclaration_12 {
   class=MaterialExpressionNamedRerouteDeclaration
   inputs [None {from=MaterialExpressionGetMaterialAttributes_1, output=BaseColor}]
   properties {name=BaseColor_In, nodeColor=<rgba 0 0.8828 1 1>}
  }
  MaterialExpressionNamedRerouteDeclaration_13 {
   class=MaterialExpressionNamedRerouteDeclaration
   inputs [None {from=MaterialExpressionScalarParameter_5}]
   properties {name=Roughness, nodeColor=<rgba 1 0.4922 0 1>}
  }
  MaterialExpressionNamedRerouteDeclaration_14 {
   class=MaterialExpressionNamedRerouteDeclaration
   inputs [None {from=MaterialExpressionScalarParameter_19}]
   properties {name=Metallic, nodeColor=<rgba 1 0.4922 0 1>}
  }
  MaterialExpressionNamedRerouteDeclaration_15 {
   class=MaterialExpressionNamedRerouteDeclaration
   inputs [None {from=MaterialExpressionGetMaterialAttributes_1, output=Metallic}]
   properties {name=Metallic_In, nodeColor=<rgba 0 0.8828 1 1>}
  }
  MaterialExpressionNamedRerouteDeclaration_16 {
   class=MaterialExpressionNamedRerouteDeclaration
   inputs [None {from=MaterialExpressionGetMaterialAttributes_1, output=Specular}]
   properties {name=Specular_In, nodeColor=<rgba 0 0.8828 1 1>}
  }
  MaterialExpressionNamedRerouteDeclaration_17 {
   class=MaterialExpressionNamedRerouteDeclaration
   inputs [None {from=MaterialExpressionGetMaterialAttributes_1, output=Roughness}]
   properties {name=Roughness_In, nodeColor=<rgba 0 0.8828 1 1>}
  }
  MaterialExpressionNamedRerouteDeclaration_18 {
   class=MaterialExpressionNamedRerouteDeclaration
   inputs [None {from=MaterialExpressionGetMaterialAttributes_1, output=EmissiveColor}]
   properties {name=Emissive_In, nodeColor=<rgba 1 0 0.1641 1>}
  }
  MaterialExpressionNamedRerouteDeclaration_19 {
   class=MaterialExpressionNamedRerouteDeclaration
   inputs [None {from=MaterialExpressionFunctionInput_2}]
   properties {name=TechArtMask, nodeColor=<rgba 0.2422 1 0 1>}
  }
  MaterialExpressionNamedRerouteDeclaration_20 {
   class=MaterialExpressionNamedRerouteDeclaration
   inputs [None {from=MaterialExpressionGetMaterialAttributes_1, output=MaterialAttributes}]
   properties {name=Mat_In, nodeColor=<rgba 1 0.1406 0 1>}
  }
  MaterialExpressionNamedRerouteDeclaration_25 {
   class=MaterialExpressionNamedRerouteDeclaration
   inputs [None {from=MaterialExpressionTextureSample_0, output=R}]
   properties {name=CausticsDistortionGradient, nodeColor=<rgba 0 0.1797 1 1>}
  }
  MaterialExpressionNamedRerouteDeclaration_26 {
   class=MaterialExpressionNamedRerouteDeclaration
   inputs [None {from=MaterialExpressionFunctionInput_0}]
   properties {name=PreskinnedLocalPosition, nodeColor=<rgba 1 0.8672 0 1>}
  }
  MaterialExpressionNamedRerouteDeclaration_28 {
   class=MaterialExpressionNamedRerouteDeclaration
   inputs [None {from=MaterialExpressionMultiply_41}]
   properties {name=Bubbles, nodeColor=<rgba 0 1 0.5547 1>}
  }
  MaterialExpressionNamedRerouteDeclaration_29 {
   class=MaterialExpressionNamedRerouteDeclaration
   inputs [None {from=MaterialExpressionLinearInterpolate_9}]
   properties {name=FinalOpacity, nodeColor=<rgba 1 0.2344 0 1>}
  }
  MaterialExpressionNamedRerouteDeclaration_30 {
   class=MaterialExpressionNamedRerouteDeclaration
   inputs [None {from=MaterialExpressionLinearInterpolate_11}]
   properties {name=FinalBasecolor, nodeColor=<rgba 0.5234 1 0 1>}
  }
  MaterialExpressionNamedRerouteDeclaration_31 {
   class=MaterialExpressionNamedRerouteDeclaration
   inputs [None {from=MaterialExpressionLinearInterpolate_14}]
   properties {name=BlurDistance, nodeColor=<rgba 0 1 0.9766 1>}
  }
  MaterialExpressionNamedRerouteDeclaration_32 {
   class=MaterialExpressionNamedRerouteDeclaration
   inputs [None {from=MaterialExpressionComponentMask_5}]
   properties {name=CameraDist, nodeColor=<rgba 0 1 0.5781 1>}
  }
  MaterialExpressionNamedRerouteDeclaration_33 {
   class=MaterialExpressionNamedRerouteDeclaration
   inputs [None {from=MaterialExpressionFresnel_0}]
   properties {name=Fresnel, nodeColor=<rgba 1 0.3984 0 1>}
  }
  MaterialExpressionNamedRerouteDeclaration_34 {
   class=MaterialExpressionNamedRerouteDeclaration
   inputs [None {from=MaterialExpressionShaderStageSwitch_1}]
   properties {name=PreskinnedLocalNormal, nodeColor=<rgba 0 1 0.9063 1>}
  }
  MaterialExpressionNamedRerouteDeclaration_4 {
   class=MaterialExpressionNamedRerouteDeclaration
   inputs [None {from=MaterialExpressionFeatureLevelSwitch_3}]
   properties {name=Emissive, nodeColor=<rgba 0.2656 0 1 1>}
  }
  MaterialExpressionNamedRerouteDeclaration_5 {
   class=MaterialExpressionNamedRerouteDeclaration
   inputs [None {from=MaterialExpressionMultiply_11}]
   properties {name=ThinFilm, nodeColor=<rgba 0.7812 0 1 1>}
  }
  MaterialExpressionNamedRerouteDeclaration_6 {
   class=MaterialExpressionNamedRerouteDeclaration
   inputs [None {from=MaterialExpressionDesaturation_1}]
   properties {name=BlurredColor, nodeColor=<rgba 0.0547 1 0 1>}
  }
  MaterialExpressionNamedRerouteDeclaration_7 {
   class=MaterialExpressionNamedRerouteDeclaration
   inputs [None {from=MaterialExpressionFeatureLevelSwitch_2}]
   properties {name=BaseColor, nodeColor=<rgba 1 0 0.375 1>}
  }
  MaterialExpressionNamedRerouteDeclaration_8 {
   class=MaterialExpressionNamedRerouteDeclaration
   inputs [None {from=MaterialExpressionFeatureLevelSwitch_1}]
   properties {name=Opacity, nodeColor=<rgba 0 0.6953 1 1>}
  }
  MaterialExpressionNamedRerouteDeclaration_9 {
   class=MaterialExpressionNamedRerouteDeclaration
   inputs [None {from=MaterialExpressionScalarParameter_2}]
   properties {name=Specular, nodeColor=<rgba 1 0 0.8906 1>}
  }
  MaterialExpressionNamedRerouteUsage_10 {class=MaterialExpressionNamedRerouteUsage}
  MaterialExpressionNamedRerouteUsage_11 {class=MaterialExpressionNamedRerouteUsage}
  MaterialExpressionNamedRerouteUsage_13 {class=MaterialExpressionNamedRerouteUsage}
  MaterialExpressionNamedRerouteUsage_15 {class=MaterialExpressionNamedRerouteUsage}
  MaterialExpressionNamedRerouteUsage_24 {class=MaterialExpressionNamedRerouteUsage}
  MaterialExpressionNamedRerouteUsage_25 {class=MaterialExpressionNamedRerouteUsage}
  MaterialExpressionNamedRerouteUsage_28 {class=MaterialExpressionNamedRerouteUsage}
  MaterialExpressionNamedRerouteUsage_29 {class=MaterialExpressionNamedRerouteUsage}
  MaterialExpressionNamedRerouteUsage_30 {class=MaterialExpressionNamedRerouteUsage}
  MaterialExpressionNamedRerouteUsage_31 {class=MaterialExpressionNamedRerouteUsage}
  MaterialExpressionNamedRerouteUsage_32 {class=MaterialExpressionNamedRerouteUsage}
  MaterialExpressionNamedRerouteUsage_34 {class=MaterialExpressionNamedRerouteUsage}
  MaterialExpressionNamedRerouteUsage_36 {class=MaterialExpressionNamedRerouteUsage}
  MaterialExpressionNamedRerouteUsage_5 {class=MaterialExpressionNamedRerouteUsage}
  MaterialExpressionNamedRerouteUsage_6 {class=MaterialExpressionNamedRerouteUsage}
  MaterialExpressionNamedRerouteUsage_7 {class=MaterialExpressionNamedRerouteUsage}
  MaterialExpressionNamedRerouteUsage_8 {class=MaterialExpressionNamedRerouteUsage}
  MaterialExpressionNamedRerouteUsage_9 {class=MaterialExpressionNamedRerouteUsage}
  MaterialExpressionNormalize_1 {class=MaterialExpressionNormalize, inputs [VectorInput {from=MaterialExpressionSubtract_2, output=""}]}
  MaterialExpressionOneMinus_0 {class=MaterialExpressionOneMinus, inputs [None {from=MaterialExpressionReroute_3}]}
  MaterialExpressionOneMinus_2 {class=MaterialExpressionOneMinus, inputs [None {from=MaterialExpressionNamedRerouteUsage_34}]}
  MaterialExpressionOneMinus_3 {class=MaterialExpressionOneMinus, inputs [None {from=MaterialExpressionAbs_0}]}
  MaterialExpressionPanner_1 {
   class=MaterialExpressionPanner
   inputs [
    Coordinate {from=MaterialExpressionMultiply_43}
    Time {from=""}
    Speed {from=MaterialExpressionAppendVector_2}
   ]
  }
  MaterialExpressionPanner_3 {
   class=MaterialExpressionPanner
   inputs [
    Coordinate {from=MaterialExpressionConvert_3}
    Time {from=""}
    Speed {from=""}
   ]
   properties {speedX=0.053, speedY=-0.1124}
  }
  MaterialExpressionPower_2 {
   class=MaterialExpressionPower
   inputs [
    Base {from=MaterialExpressionOneMinus_0}
    Exp {from=""}
   ]
   properties {constExponent=4}
  }
  MaterialExpressionPreSkinnedNormal_0 {class=MaterialExpressionPreSkinnedNormal, inputs [], properties {}}
  MaterialExpressionQualitySwitch_0 {
   class=MaterialExpressionQualitySwitch
   inputs [
    Default {from=MaterialExpressionAdd_15, output=""}
    Low {from=MaterialExpressionReroute_2, output=""}
    High {from="", output=""}
    Medium {from="", output=""}
    Epic {from="", output=""}
   ]
   properties {}
  }
  MaterialExpressionQualitySwitch_1 {
   class=MaterialExpressionQualitySwitch
   inputs [
    Default {from=MaterialExpressionScalarParameter_51, output=""}
    Low {from="", output=""}
    High {from="", output=""}
    Medium {from="", output=""}
    Epic {from=MaterialExpressionScalarParameter_46, output=""}
   ]
   properties {}
  }
  MaterialExpressionQualitySwitch_2 {
   class=MaterialExpressionQualitySwitch
   inputs [
    Default {from=MaterialExpressionConstant_9, output=""}
    Low {from=MaterialExpressionConstant_11, output=""}
    High {from=MaterialExpressionConstant_10, output=""}
    Medium {from=MaterialExpressionConstant_11, output=""}
    Epic {from=MaterialExpressionConstant_10, output=""}
   ]
   properties {}
  }
  MaterialExpressionReroute_2 {class=MaterialExpressionReroute, inputs [None {from=MaterialExpressionNamedRerouteUsage_28}]}
  MaterialExpressionReroute_3 {class=MaterialExpressionReroute, inputs [None {from=MaterialExpressionMaterialFunctionCall_6, output=XY Texture}]}
  MaterialExpressionSaturate_0 {class=MaterialExpressionSaturate, inputs [None {from=MaterialExpressionDivide_7}]}
  MaterialExpressionScalarParameter_0 {class=MaterialExpressionScalarParameter, properties {defaultValue=1, group=FunkyFeather Beak, parameterName=ViewDependantDistortion}}
  MaterialExpressionScalarParameter_14 {class=MaterialExpressionScalarParameter, properties {defaultValue=2, group=FunkyFeather Beak, parameterName=ThinFilmExp}}
  MaterialExpressionScalarParameter_15 {class=MaterialExpressionScalarParameter, properties {defaultValue=0.5, group=FunkyFeather Beak, parameterName=MobileOpacity, sortPriority=2}}
  MaterialExpressionScalarParameter_17 {class=MaterialExpressionScalarParameter, properties {defaultValue=1, group=TransparentArmor, parameterName=Emissive_Final_Day}}
  MaterialExpressionScalarParameter_18 {class=MaterialExpressionScalarParameter, properties {defaultValue=0.5, group=TransparentArmor, parameterName=Emissive_Final_Night}}
  MaterialExpressionScalarParameter_19 {class=MaterialExpressionScalarParameter, properties {group=FunkyFeather Beak, parameterName=Metallic, sortPriority=1}}
  MaterialExpressionScalarParameter_2 {class=MaterialExpressionScalarParameter, properties {defaultValue=1, group=FunkyFeather Beak, parameterName=Specular, sortPriority=1}}
  MaterialExpressionScalarParameter_26 {class=MaterialExpressionScalarParameter, properties {defaultValue=1, group=FunkyFeather Beak, parameterName=ThinFilmMult}}
  MaterialExpressionScalarParameter_28 {class=MaterialExpressionScalarParameter, properties {group=FunkyFeather Beak, parameterName=water_SpeedX}}
  MaterialExpressionScalarParameter_29 {class=MaterialExpressionScalarParameter, properties {defaultValue=0.1, group=FunkyFeather Beak, parameterName=water_SpeedY}}
  MaterialExpressionScalarParameter_4 {class=MaterialExpressionScalarParameter, properties {defaultValue=0.9, group=FunkyFeather Beak, parameterName=DistortionIntensity}}
  MaterialExpressionScalarParameter_40 {class=MaterialExpressionScalarParameter, properties {defaultValue=20, group=FunkyFeather Beak, parameterName=Bubble Size}}
  MaterialExpressionScalarParameter_41 {class=MaterialExpressionScalarParameter, properties {defaultValue=15, group=FunkyFeather Beak, parameterName=Bubble ParallaxDepth}}
  MaterialExpressionScalarParameter_42 {class=MaterialExpressionScalarParameter, properties {defaultValue=3, group=FunkyFeather Beak, parameterName=Bubble Distortion}}
  MaterialExpressionScalarParameter_43 {class=MaterialExpressionScalarParameter, properties {defaultValue=3, group=FunkyFeather Beak, parameterName=Bubble Visbility}}
  MaterialExpressionScalarParameter_44 {class=MaterialExpressionScalarParameter, properties {defaultValue=2, group=FunkyFeather Beak, parameterName=Bubble ProjectionContrast}}
  MaterialExpressionScalarParameter_45 {class=MaterialExpressionScalarParameter, properties {defaultValue=0.75, group=FunkyFeather Beak, parameterName=FresnelExp}}
  MaterialExpressionScalarParameter_46 {class=MaterialExpressionScalarParameter, properties {defaultValue=0.001, group=FunkyFeather Beak, parameterName=BlurNearEpic}}
  MaterialExpressionScalarParameter_47 {class=MaterialExpressionScalarParameter, properties {group=FunkyFeather Beak, parameterName=BlurDistanceBias}}
  MaterialExpressionScalarParameter_48 {class=MaterialExpressionScalarParameter, properties {defaultValue=600, group=FunkyFeather Beak, parameterName=BlurDistanceFade}}
  MaterialExpressionScalarParameter_49 {class=MaterialExpressionScalarParameter, properties {defaultValue=0.0003, group=FunkyFeather Beak, parameterName=BlurFar}}
  MaterialExpressionScalarParameter_5 {class=MaterialExpressionScalarParameter, properties {defaultValue=0.2, group=FunkyFeather Beak, parameterName=Roughness, sortPriority=1}}
  MaterialExpressionScalarParameter_50 {class=MaterialExpressionScalarParameter, properties {defaultValue=1, group=FunkyFeather Beak, parameterName=BlurUseDither}}
  MaterialExpressionScalarParameter_51 {class=MaterialExpressionScalarParameter, properties {defaultValue=0.001, group=FunkyFeather Beak, parameterName=BlurNearDefault}}
  MaterialExpressionScalarParameter_52 {class=MaterialExpressionScalarParameter, properties {defaultValue=2, group=FunkyFeather Beak, parameterName=BlurDitherMulitplier}}
  MaterialExpressionScalarParameter_53 {class=MaterialExpressionScalarParameter, properties {defaultValue=-0.5, group=FunkyFeather Beak, parameterName=BlurredColor_Desat}}
  MaterialExpressionScalarParameter_6 {
   class=MaterialExpressionScalarParameter
   properties {defaultValue=1, group=FunkyFeather Beak, parameterName=Opacity_SceneColor, sortPriority=2}
  }
  MaterialExpressionScalarParameter_9 {class=MaterialExpressionScalarParameter, properties {defaultValue=0.6, group=FunkyFeather Beak, parameterName=ThinFilmDesaturation}}
  MaterialExpressionSceneColor_4 {class=MaterialExpressionSceneColor, inputs [UVs {from="", output=""}], properties {}}
  MaterialExpressionSetMaterialAttributes_1 {
   class=MaterialExpressionSetMaterialAttributes
   inputs [
    MaterialAttributes {from=MaterialExpressionNamedRerouteDeclaration_20, output=""}
    BaseColor {from=MaterialExpressionNamedRerouteUsage_32, output=""}
    Metallic {from=MaterialExpressionNamedRerouteUsage_15, output=""}
    Specular {from=MaterialExpressionNamedRerouteUsage_11, output=""}
    Roughness {from=MaterialExpressionNamedRerouteUsage_13, output=""}
    EmissiveColor {from=MaterialExpressionNamedRerouteUsage_9, output=""}
    Opacity {from=MaterialExpressionNamedRerouteUsage_31, output=""}
    Normal {from="", output=""}
   ]
   properties {attributeSetTypes [
     #1
     #2
     #3
     #4
     #5
     #7
     #6
    ]}
  }
  MaterialExpressionShaderStageSwitch_0 {
   class=MaterialExpressionShaderStageSwitch
   inputs [
    PixelShader {from=MaterialExpressionVertexInterpolator_0, output=PS}
    VertexShader {from=MaterialExpressionLocalPosition_1, output=XYZ}
   ]
   properties {}
  }
  MaterialExpressionShaderStageSwitch_1 {
   class=MaterialExpressionShaderStageSwitch
   inputs [
    PixelShader {from=MaterialExpressionVertexInterpolator_1, output=PS}
    VertexShader {from=MaterialExpressionPreSkinnedNormal_0, output=""}
   ]
   properties {}
  }
  MaterialExpressionShadingPathSwitch_0 {
   class=MaterialExpressionShadingPathSwitch
   inputs [
    Default {from=MaterialExpressionQualitySwitch_0, output=""}
    Deferred {from=MaterialExpressionQualitySwitch_0, output=""}
    Forward {from=MaterialExpressionReroute_2, output=""}
    Mobile {from=MaterialExpressionReroute_2, output=""}
   ]
   properties {}
  }
  MaterialExpressionSine_0 {class=MaterialExpressionSine, inputs [None {from=MaterialExpressionAdd_8}]}
  MaterialExpressionSmoothStep_1 {
   class=MaterialExpressionSmoothStep
   inputs [
    Min {from="", output=""}
    Max {from="", output=""}
    Value {from=MaterialExpressionOneMinus_3, output=""}
   ]
   properties {constMax=0.5, constMin=0.1}
  }
  MaterialExpressionSubtract_2 {
   class=MaterialExpressionSubtract
   inputs [
    A {from=MaterialExpressionNamedRerouteUsage_28}
    B {from=MaterialExpressionTransformPosition_0}
   ]
  }
  MaterialExpressionSubtract_4 {
   class=MaterialExpressionSubtract
   inputs [
    A {from=MaterialExpressionComponentMask_5}
    B {from=MaterialExpressionScalarParameter_47}
   ]
  }
  MaterialExpressionTextureObject_0 {
   class=MaterialExpressionTextureObject
   inputs []
   properties {
    samplerType=SAMPLERTYPE_Masks
    texture=/BRCosmetics/Characters/Player/Male/Medium/Bodies/M_MED_FunkyFeather_Cap/Skins/Beak/Textures/T_FunkyFeather_CapBeak_TA
   }
  }
  MaterialExpressionTextureSampleParameter2D_0 {
   class=MaterialExpressionTextureSampleParameter2D
   inputs [
    UVs {from=MaterialExpressionFresnel_1, output=""}
    Apply View MipBias {from="", output=""}
   ]
   properties {group=TransparentArmor, parameterName=ThinFilmTexture, texture=/Game/Global/Textures/Default/Thinfilm/T_ThinFilm_Spectrum_CoolColor}
  }
  MaterialExpressionTextureSample_0 {
   class=MaterialExpressionTextureSample
   inputs [
    UVs {from=MaterialExpressionAdd_7}
    Tex {from=MaterialExpressionTextureObject_0}
    Apply View MipBias {from=""}
   ]
   properties {samplerSource=SSM_Wrap_WorldGroupSettings, samplerType=SAMPLERTYPE_Masks}
  }
  MaterialExpressionTextureSample_2 {
   class=MaterialExpressionTextureSample
   inputs [
    UVs {from=MaterialExpressionPanner_1}
    Tex {from=MaterialExpressionTextureObject_0}
    Apply View MipBias {from=""}
   ]
   properties {samplerType=SAMPLERTYPE_LinearColor, texture=/Game/Characters/Player/Male/Medium/Heads/M_MED_AUX_Alien_Head_01/Textures/T_DistortionMask}
  }
  MaterialExpressionTime_0 {class=MaterialExpressionTime}
  MaterialExpressionTransformPosition_0 {
   class=MaterialExpressionTransformPosition
   inputs [None {from=MaterialExpressionCameraPositionWS_1, output=""}]
   properties {transformSourceType=TRANSFORMPOSSOURCE_World}
  }
  MaterialExpressionTransformPosition_1 {
   class=MaterialExpressionTransformPosition
   inputs [None {from=MaterialExpressionWorldPosition_1, output=XYZ}]
   properties {transformSourceType=TRANSFORMPOSSOURCE_World, transformType=TRANSFORMPOSSOURCE_View}
  }
  MaterialExpressionVectorParameter_0 {
   class=MaterialExpressionVectorParameter
   properties {group=TransparentArmor, parameterName=SurfaceEmissiveColor, sortPriority=3, defaultValue=<rgba 0.7588 0.974 0.9553 1>}
  }
  MaterialExpressionVectorParameter_1 {
   class=MaterialExpressionVectorParameter
   properties {group=FunkyFeather Beak, parameterName=SurfaceBaseColor, sortPriority=5, defaultValue=<rgba 0.4271 0.4566 0.6406 1>}
  }
  MaterialExpressionVectorParameter_2 {
   class=MaterialExpressionVectorParameter
   properties {group=FunkyFeather Beak, parameterName=SurfaceBaseColor_LowEnd, sortPriority=5, defaultValue=<rgba 0.4271 0.4566 0.6406 1>}
  }
  MaterialExpressionVectorParameter_3 {
   class=MaterialExpressionVectorParameter
   properties {group=TransparentArmor, parameterName=SurfaceEmissiveColor_LowEnd, sortPriority=3, defaultValue=<rgba 0.0938 0.1002 0.1406 1>}
  }
  MaterialExpressionVectorParameter_6 {
   class=MaterialExpressionVectorParameter
   properties {group=FunkyFeather Beak, parameterName=BubbleColor, defaultValue=<rgba 1.4054 1.5 0.4217 1>}
  }
  MaterialExpressionVectorParameter_8 {
   class=MaterialExpressionVectorParameter
   properties {group=FunkyFeather Beak, parameterName=FresnelColor, defaultValue=<rgba 0.2656 0.2656 0.2656 1>}
  }
  MaterialExpressionVertexInterpolator_0 {class=MaterialExpressionVertexInterpolator, inputs [VS {from=MaterialExpressionLocalPosition_1, output=XYZ}], properties {}}
  MaterialExpressionVertexInterpolator_1 {class=MaterialExpressionVertexInterpolator, inputs [VS {from=MaterialExpressionPreSkinnedNormal_0, output=""}], properties {}}
  MaterialExpressionWorldPosition_1 {class=MaterialExpressionWorldPosition, inputs [], properties {}}
 }}
