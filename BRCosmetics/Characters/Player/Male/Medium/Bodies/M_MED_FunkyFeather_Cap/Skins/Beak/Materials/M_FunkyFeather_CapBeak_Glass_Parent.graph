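# Translucent parent material for the glass on the Beak skin of the FunkyFeather Cap player character, using surface per-pixel lighting and material attributes.
# It chains three material functions, starting with the 2023 character parent PreFX variant with skin disabled from the transparent-armor parent folder.
# That output feeds the skin-specific CapBeak glass function, whose result passes into the shared character PostFX function.
# It supports skeletal meshes, morph targets and Nanite.
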Material /BRCosmetics/Characters/Player/Male/Medium/Bodies/M_MED_FunkyFeather_Cap/Skins/Beak/Materials/M_FunkyFeather_CapBeak_Glass_Parent
graph {
 properties {
  bUseMaterialAttributes=true
  bUsedWithMorphTargets=true
  bUsedWithNanite=true
  bUsedWithSkeletalMesh=true
  blendMode=BLEND_Translucent
  pixelDepthOffsetMode=PDOM_Legacy
  shadingModel=MSM_FromMaterialExpression
  translucencyLightingMode=TLM_SurfacePerPixelLighting
  usedShadingModels=Default Lit | Subsurface | Clear Coat | Subsurface Profile | Cloth
 }
 nodes [
  n1 = MaterialFunctionCall!(Overide Texture Set UVs, Override Tech Art Mask UVs, Override Basic Emissive UVs) {materialFunction=/Game/Global/Materials/Character_Parent_2023/TransparentArmor_Parent_2025/MF_CharacterParent_2023_PreFX_SkinDisabled}
  n2 = MaterialFunctionCall!(Material=@n1.Result, PreskinnedLocPos=@n1.Result, TechArtMask=@n1.Result) {materialFunction=&/MF_FunkyFeather_CapBeak_Glass}
  MaterialFunctionCall(Material=@n2.Result, M=@n1.M, Pre Skinned Local Position=@n1.M, Distance Blend=@n1.M, Fresnel=@n1.M) {materialFunction=&MF_CharacterParent_2023_PostFX}
 ]
}
