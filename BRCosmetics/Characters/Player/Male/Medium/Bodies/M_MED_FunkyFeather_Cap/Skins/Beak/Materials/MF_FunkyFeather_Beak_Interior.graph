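# Material function that restyles the masked beak interior of the FunkyFeather character skin.
# It blends between shaded and lit emissive colors using the saturated dot product of the anime shader's light direction and the transformed normal, faking directional lighting.
# The Mask input blends in that emissive, zero specular, overridden metallic and base color scaled by Basemulti.
# A masked roughness blend exists but is not connected to the output.
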MaterialFunction /BRCosmetics/Characters/Player/Male/Medium/Bodies/M_MED_FunkyFeather_Cap/Skins/Beak/Materials/MF_FunkyFeather_Beak_Interior
graph {nodes [
  n1 = GetMaterialAttributes(None=FunctionInput() {inputName=Material, inputType=FunctionInput_MaterialAttributes}) {attributeGetTypes [
    #1
    #2
    #3
    #4
    #5
    #6
   ]}
  n2 = FunctionInput() {inputName=Mask, inputType=FunctionInput_Scalar, sortPriority=1}
  n3 = VectorParameter() {group=FunkyFeather Beak, parameterName=FFB_EmissiveColor_Shaded, defaultValue=<rgba 0.0977 0.0674 0.2696 1>}
  n4 = VectorParameter() {group=FunkyFeather Beak, parameterName=FFB_EmissiveColor, defaultValue=<rgba 0.2699 0.2036 1 1>}
  n5 = Saturate(None=DotProduct(
    A=MaterialFunctionCall() {materialFunction=/Game/Global/Materials/Character_Parent_2023/Anime_Parent_2023/MaterialFunctions/MFC_AnimeShader_LightDirection}.Direction Vector
    B=Transform(None=@n1.Normal)
   ))
  n6 = ScalarParameter() {group=FunkyFeather Beak, parameterName=Basemulti}
  n7 = SetMaterialAttributes!(
   MaterialAttributes=@n1.MaterialAttributes
   Specular=LinearInterpolate(A=@n1.Specular, Alpha=@n2) {constB=0}
   Metallic=LinearInterpolate(A=@n1.Metallic, Alpha=@n2)
   EmissiveColor=LinearInterpolate(A=@n1.EmissiveColor, B=LinearInterpolate(A=@n3.RGB, B=@n4.RGB, Alpha=@n5), Alpha=@n2)
   Normal
   BaseColor=LinearInterpolate(A=@n1.BaseColor, B=Multiply(A=@n1.BaseColor, B=@n6), Alpha=@n2) {constB=0}
  ) {attributeSetTypes [
    #2
    #3
    #1
    #6
    #5
   ]}
  FunctionOutput(None=@n7)
  LinearInterpolate(A=@n1.Roughness, B=ScalarParameter() {defaultValue=1, parameterName=Roughness}, Alpha=@n2) {constB=0.2}
 ]}
