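# Parent character material for the FunkyFeather Cap Beak skin, with an IsHead switch selecting head or body shading.
# Between the CharacterParent PreFX and PostFX functions, the head branch blends in a subsurface layer, adjusts shadow opacity mask and adds an emissive tint; both branches pass through a masked beak-interior function.
# It also builds flipbook face attributes, toon sphere-projected normals, and temporal responsiveness driven by face alpha and AnimBP-set pose changes.
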Material /BRCosmetics/Characters/Player/Male/Medium/Bodies/M_MED_FunkyFeather_Cap/Skins/Beak/Materials/M_FunkyFeather_CapBeak_Head_Parent
graph {
 properties {
  bUseMaterialAttributes=true
  bUsedWithClothing=true
  bUsedWithMeshDeformer=true
  bUsedWithMorphTargets=true
  bUsedWithNanite=true
  bUsedWithNiagaraMeshParticles=true
  bUsedWithSkeletalMesh=true
  pixelDepthOffsetMode=PDOM_Legacy
  shadingModel=MSM_FromMaterialExpression
  usedShadingModels=Default Lit | Subsurface | Clear Coat | Subsurface Profile | Cloth
 }
 nodes [
  LinearInterpolate(B=Step(Y=VertexColor().R), Alpha=ScalarParameter() {parameterName=IsInFE})
  n1 = MaterialFunctionCall() {materialFunction=&MF_CharacterParent_2023_PreFX}
  n2 = NamedRerouteUsage()
  n3 = Constant3Vector() {constant=<rgba 0 0 1 0>}
  n4 = Constant() {r=1}
  n5 = ShadingModel() {shadingModel=MSM_Subsurface}
  n6 = NamedRerouteUsage()
  n7 = GetMaterialAttributes(None=BlendMaterialAttributes(
    A=@n1.Result
    B=SetMaterialAttributes!(MaterialAttributes=@n2, Normal=@n3, Opacity=@n4, Shading Model=@n5) {attributeSetTypes [#1, #2, #3]}
    Alpha=@n6
   )) {attributeGetTypes [#4, #1]}
  n8 = ComponentMask(None=@n1.Tech Art Mask) {b=true}
  n9 = Saturate(None=Add(A=@n6, B=Step(Y=@n8) {constX=0.5, constY=0.5}))
  n10 = LinearInterpolate(A=@n7.Opacity Mask, B=ShadowReplace(Default=@n7.Opacity Mask, Shadow=Constant()), Alpha=@n9)
  n11 = NamedRerouteUsage()
  n12 = OneMinus(None=@n9)
  n13 = ChannelMaskParameter(None=@n1.Tech Art Mask) {group=FunkyFeather, maskChannel=Blue, parameterName=InteriorMask, defaultValue=<rgba 0 0 1 0>}
  n14 = LinearInterpolate(
   A=GetMaterialAttributes(None=@n1.Result) {attributeGetTypes [#4, #1]}.Normal
   B=Constant3Vector() {constant=<rgba 0 0 1 0>}
   Alpha=NamedRerouteDeclaration(None=StaticSwitchParameter(True=@n12, False=@n13.R) {defaultValue=true, desc=Head or body, parameterName=IsHead}) {name=InteriorMask, nodeColor=<rgba 0 0.4844 1 1>}
  )
  n15 = NamedRerouteUsage()
  n16 = MaterialFunctionCall!(
   Material=StaticSwitchParameter(
    True=SetMaterialAttributes!(MaterialAttributes=@n7.MaterialAttributes, Opacity Mask=@n10, Normal=@n11) {attributeSetTypes [#4, #1]}
    False=SetMaterialAttributes(MaterialAttributes=@n1.Result, Normal=@n14) {desc=body normal, attributeSetTypes [#4, #1]}
   ) {defaultValue=true, desc=Head or body, parameterName=IsHead}
   Mask=@n15
  ) {materialFunction=&/MF_FunkyFeather_Beak_Interior}
  n17 = ComponentMask(None=NamedRerouteUsage()) {b=true}
  n18 = LinearInterpolate(
   A=GetMaterialAttributes(None=@n16.Result) {attributeGetTypes [#5]}.Emissive Color
   B=VectorParameter() {group=FunkyFeather Beak, parameterName=FFB_EmissiveColor2, defaultValue=<rgba 0.1925 0.1451 0.713 1>}.RGB
   Alpha=Multiply(A=SmoothStep(Value=@n17) {constMax=143, constMin=153}, B=@n15)
  )
  MaterialFunctionCall(
   Material=StaticSwitchParameter(
    True=SetMaterialAttributes!(MaterialAttributes=@n16.Result, Emissive Color=@n18) {attributeSetTypes [#5]}
    False=@n16.Result
   ) {defaultValue=true, desc=Head or body, parameterName=IsHead}
   M=@n1.M
   Pre Skinned Local Position=@n1.M
   Distance Blend=@n1.M
   Fresnel=@n1.M
  ) {materialFunction=&MF_CharacterParent_2023_PostFX}
  n19 = Subtract(A=TextureCoordinate(), B=Constant2Vector() {r=1})
  n20 = ComponentMask(None=NamedRerouteDeclaration(None=Saturate(None=@n19)) {name=FaceUVs, nodeColor=<rgba 1 0 0.2578 1>}) {r=true}
  NamedRerouteDeclaration(None=Saturate(None=Sign(None=@n20))) {name=EyesBrowsMask, nodeColor=<rgba 0.875 1 0 1>}
  NamedRerouteDeclaration(None=@n7.Normal) {name=Normal_IN, nodeColor=<rgba 0.5 1 0 1>}
  n21 = Transform(None=NamedRerouteUsage())
  n22 = Subtract(
   A=VectorParameter() {group=Toon_Shading, parameterName=NormalSphere_Location, defaultValue=<rgba 0 0 162 1>}.RGB
   B=LocalPosition() {localOrigin=InstancePreSkinning}.XYZ
  )
  n23 = LinearInterpolate(
   A=@n22
   B=Multiply(A=@n22, B=Constant3Vector() {constant=<rgba 1 1 0 0>})
   Alpha=ScalarParameter() {group=Toon_Shading, parameterName=NormalSphereCylindrical_Blend, sliderMax=1}
  )
  n24 = VertexInterpolator(VS=Transform(None=Normalize(VectorInput=@n23)) {transformSourceType=TRANSFORMSOURCE_Local})
  n25 = Multiply(A=ScalarParameter() {group=Toon_Shading, parameterName=NormalSphere_Blend, sliderMax=1}, B=NamedRerouteUsage())
  NamedRerouteDeclaration(None=Transform(None=LinearInterpolate(A=@n21, B=@n24.PS, Alpha=@n25) {constAlpha=0}) {transformSourceType=TRANSFORMSOURCE_World, transformType=TRANSFORM_Tangent}) {name=SphereNormal_OUT, nodeColor=<rgba 0.6406 1 0 1>}
  NamedRerouteDeclaration(None=@n1.Pre skinned Local Position) {name=PreskinnedLocPos, nodeColor=<rgba 0 0.9766 1 1>}
  n26 = NamedRerouteUsage()
  n27 = MaterialFunctionCall!(Mouth UV, Eye UV=@n26, Brow UV=@n26, Material=@n1.Result) {materialFunction=/Game/Global/Materials/Character_Parent_2023/FlipbookFace_Parent_2023/MaterialFunctions/MF_FlipbookFace}
  NamedRerouteDeclaration(None=@n27.Material) {name=FlipbookFace_MA, nodeColor=<rgba 0.8984 0 1 1>}
  n28 = ScalarParameter() {desc="Set by base head AnimBP if \"IsFlipbookFace\" is true", parameterName=HasPoseChanged}
  TemporalResponsivenessOutput!(None=LinearInterpolate(Alpha=Multiply(A=@n27.FaceAlpha, B=@n28)) {constA=0.4})
 ]
}
